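# Material function for box height fog, outputting material attributes with a flat emissive Color, depth-based opacity and world position offset.
# Opacity sums how far the scene behind translucency and the camera sit below the scale-adjusted actor height, over FadeDistance.
# A camera test against scaled box extents collapses vertices to the object pivot, with a static bool selecting outside or inside culling.
# Lava fake-light, world-aligned noise and edge-glow networks remain unconnected to the output.
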MaterialFunction /Figment_S09/Environments/Sets/Vulcano/Materials/MF_BoxHeightFog
graph {nodes [
  n1 = ScalarParameter() {defaultValue=95.3291, parameterName=FakeLightHeight}
  Add(A=Multiply(A=Constant3Vector() {constant=<rgba 0 0 1 0>}, B=@n1))
  n2 = MaterialFunctionCall() {materialFunction=&WorldPositionBehindTranslucency}
  n3 = Subtract(A=ComponentMask(None=@n2.Result) {b=true}, B=ComponentMask(None=ActorPositionWS()) {b=true})
  Clamp(None=Divide(A=@n3, B=ScalarParameter() {defaultValue=0.4501, parameterName=BottomMask}))
  ComponentMask(None=MaterialFunctionCall() {materialFunction=&WorldPositionBehindTranslucency}.Result) {b=true}
  Constant() {r=1}
  n4 = Constant() {r=1}
  n5 = Divide(A=@n3, B=@n1)
  n6 = Abs(None=Clamp(None=@n5) {maxDefault=0, minDefault=-1})
  n7 = Clamp(None=@n5)
  n8 = ScalarParameter() {defaultValue=0.5143, parameterName=FakeLightFalloff}
  n9 = TextureObject() {samplerType=SAMPLERTYPE_Masks, texture=/Game/Environments/Sets/Vulcano/Textures/LavaGround_M}
  n10 = ScalarParameter() {defaultValue=420.6348, parameterName=TextureSize}
  n11 = AppendVector(A=AppendVector(A=@n10, B=@n10), B=@n10)
  n12 = Add(A=Multiply(A=Time(), B=ScalarParameter() {defaultValue=-72.5004, parameterName=Speed}), B=@n2.Result)
  n13 = SceneTexture() {sceneTextureId=PPI_WorldNormal}
  n14 = ComponentMask(None=@n13.Color) {b=true, g=true, r=true}
  n15 = ScalarParameter() {parameterName=Projection}
  n16 = MaterialFunctionCall(
   Transition Phase=OneMinus(None=Power(Base=@n7, Exp=@n8))
   Height Texture=OneMinus(None=MaterialFunctionCall(TextureObject=@n9, TextureSize=@n11, WorldPosition=@n12, World Space Normal=@n14, ProjectionTransitionContrast=@n15) {materialFunction=&WorldAlignedTexture}.XYZ Texture)
   Contrast=ScalarParameter() {defaultValue=2, parameterName=Contrast}
  ) {materialFunction=&HeightLerp}
  n17 = Clamp(None=LinearInterpolate(
    A=ScalarParameter() {defaultValue=6, parameterName=ContrastMin}
    B=ScalarParameter() {defaultValue=-2.4472, parameterName=ContrastMax}
    Alpha=ComponentMask(None=@n13.Color) {b=true}
   ))
  n18 = Multiply(A=Multiply(A=OneMinus(None=@n6), B=Multiply(A=@n16.Lerp Alpha No Contrast, B=@n17)))
  n19 = ShadingPathSwitch(Default=@n18, Deferred=@n18, Forward=@n4, Mobile=@n4)
  FeatureLevelSwitch(Default=@n19, ES2=@n4, ES3_1=@n4, SM4_REMOVED=@n19)
  n20 = ComponentMask(None=MaterialFunctionCall() {materialFunction=&WorldPositionBehindTranslucency}.Result) {b=true}
  n21 = ComponentMask(None=ActorPositionWS()) {b=true}
  n22 = Multiply(
   A=ScalarParameter() {parameterName=Offset}
   B=MaterialFunctionCall() {materialFunction=/Engine/Functions/Engine_MaterialFunctions02/WorldPositionOffset/ObjectScale}.Scale Z
  )
  n23 = Multiply(A=Subtract(A=Subtract(A=@n20, B=@n21), B=@n22)) {constB=-1}
  n24 = ScalarParameter() {defaultValue=2000, parameterName=FadeDistance}
  n25 = CameraPositionWS()
  n26 = Subtract(A=Subtract(A=ComponentMask(None=@n25) {b=true}, B=@n21), B=@n22)
  n27 = Clamp(None=Divide(A=Multiply(A=@n26) {constB=-1}, B=@n24))
  n28 = Add(A=Clamp(None=Divide(A=@n23, B=@n24)), B=@n27)
  n29 = ComponentMask(None=ObjectPositionWS()) {b=true}
  n30 = ComponentMask(None=CameraPositionWS()) {b=true}
  n31 = MaterialFunctionCall() {materialFunction=&ObjectScale}
  n32 = Subtract(A=Abs(None=Subtract(A=@n29, B=@n30)), B=Multiply(A=ScalarParameter() {defaultValue=50, parameterName=FadeDistanceZ}, B=@n31.Scale Z))
  n33 = Subtract(A=ComponentMask(None=ObjectPositionWS()) {r=true}, B=ComponentMask(None=CameraPositionWS()) {r=true})
  n34 = ScalarParameter() {defaultValue=100, parameterName=FadeDistanceX}
  n35 = Floor(None=Subtract(A=Abs(None=@n33), B=Multiply(A=@n34, B=@n31.Scale X)))
  n36 = Subtract(A=ComponentMask(None=ObjectPositionWS()) {g=true}, B=ComponentMask(None=CameraPositionWS()) {g=true})
  n37 = ScalarParameter() {defaultValue=100, parameterName=FadeDistanceY}
  n38 = Floor(None=Subtract(A=Abs(None=@n36), B=Multiply(A=@n31.Scale Y, B=@n37)))
  n39 = Saturate(None=Add(A=Floor(None=@n32), B=Add(A=@n35, B=@n38)))
  n40 = Subtract(A=ObjectPositionWS(), B=WorldPosition().XYZ)
  n41 = Multiply(A=@n39, B=@n40)
  n42 = Multiply(A=@n40, B=OneMinus(None=@n39))
  n43 = FunctionInput(Preview=StaticBool()) {bUsePreviewValueAsDefault=true, inputType=FunctionInput_StaticBool}
  FunctionOutput(None=MakeMaterialAttributes!(
    BaseColor
    Metallic
    Specular
    Roughness
    Anisotropy
    EmissiveColor=VectorParameter() {parameterName=Color, defaultValue=<rgba 0.9442 0.7266 1 1>}.RGB
    Opacity=@n28
    OpacityMask
    Normal
    Tangent
    WorldPositionOffset=StaticSwitch(True=@n41, False=@n42, Value=@n43)
    SubsurfaceColor
    ClearCoat
    ClearCoatRoughness
    AmbientOcclusion
    Refraction
    CustomizedUVs_0
    CustomizedUVs_1
    CustomizedUVs_2
    CustomizedUVs_3
    CustomizedUVs_4
    CustomizedUVs_5
    CustomizedUVs_6
    CustomizedUVs_7
    PixelDepthOffset
    ShadingModel
    Displacement
   ))
  n44 = Constant3Vector()
  n45 = Add(
   A=ComponentMask(None=ActorPositionWS()) {b=true}
   B=Add(
    A=ScalarParameter() {defaultValue=10, parameterName=TransitionMargin}
    B=ScalarParameter() {defaultValue=95.3291, parameterName=FakeLightHeight}
   )
  )
  n46 = ComponentMask(None=CameraPositionWS()) {b=true}
  n47 = Ceil(None=Clamp(None=Subtract(A=@n45, B=@n46)))
  LinearInterpolate(A=@n44, B=@n40, Alpha=@n47)
  LinearInterpolate(A=@n40, B=@n44, Alpha=Multiply(A=@n47))
  n48 = ScalarParameter() {defaultValue=5, parameterName=Brightness}
  Multiply(B=ShadingPathSwitch(Default=@n48, Deferred=@n48, Forward=Multiply(A=@n48) {constB=0.35}))
  n49 = ActorPositionWS()
  n50 = Distance(A=@n49, B=CameraPositionWS())
  n51 = ScalarParameter() {defaultValue=15000, parameterName=CameraFadeOffset}
  n52 = Clamp(None=Divide(A=Subtract(A=@n50, B=@n51), B=ScalarParameter() {defaultValue=10000, parameterName=CameraFadeLength}))
  n53 = OneMinus(None=@n52)
  n54 = MaterialFunctionCall() {materialFunction=&WorldPositionBehindTranslucency}
  n55 = Distance(A=@n54.Result, B=@n49)
  n56 = ScalarParameter() {defaultValue=1359.0356, parameterName=LightRadius}
  n57 = OneMinus(None=Clamp(None=Divide(A=@n55, B=@n56)))
  n58 = Normalize(VectorInput=Subtract(A=@n49, B=@n54.Result))
  n59 = ComponentMask(None=SceneTexture() {sceneTextureId=PPI_WorldNormal}.Color) {b=true, g=true, r=true}
  n60 = LinearInterpolate(A=ScalarParameter() {defaultValue=0.1143, parameterName=MinLightTop}, Alpha=Clamp(None=DotProduct(A=@n58, B=@n59)))
  n61 = ShadingPathSwitch(Default=@n60, Deferred=@n60, Forward=ScalarParameter() {defaultValue=1, parameterName=SwitchLightBrightness}, Mobile=@n60)
  Multiply(A=@n53, B=Clamp(None=Multiply(A=@n57, B=@n61)))
  n62 = Distance(A=@n49, B=CameraPositionWS())
  n63 = ScalarParameter() {defaultValue=3579.1655, parameterName=SphereRadius}
  n64 = Ceil(None=Clamp(None=Subtract(A=@n62, B=@n63)))
  n65 = Add(A=Floor(None=@n52), B=OneMinus(None=@n64))
  Multiply(A=Subtract(A=CameraPositionWS(), B=WorldPosition().XYZ), B=StaticSwitch(True=Clamp(None=@n65), False=@n64))
  ObjectPositionWS()
  ObjectPositionWS()
  ObjectPositionWS()
  ObjectPositionWS()
  ObjectPositionWS()
  OneMinus(None=@n27)
  OneMinus(None=@n28)
  OneMinus(None=@n53)
  OneMinus()
  n66 = VectorParameter() {parameterName=LavaGlowEdgeDark, defaultValue=<rgba 0 0 0 1>}
  n67 = VectorParameter() {parameterName=LavalGlowEdgeBright, defaultValue=<rgba 20 0.8 0 1>}
  n68 = ComponentMask(None=WorldPosition().XYZ) {g=true, r=true}
  n69 = Divide(A=@n68, B=ScalarParameter() {defaultValue=968.9221, parameterName=LavaEdgeTiling1})
  n70 = ScalarParameter() {defaultValue=0.1143, parameterName=PanSpeedEdgeGlow}
  n71 = MaterialFunctionCall(XY=Divide(A=@n68, B=ScalarParameter() {defaultValue=1688.7714, parameterName=LavaEdgeTiling2})) {materialFunction=&Swizzle}
  n72 = ScalarParameter() {defaultValue=-0.1143, parameterName=PanSpeedEdgeGlow2}
  n73 = Add(
   A=TextureSample(UVs=Panner(Coordinate=@n69, Speed=@n70)) {samplerType=SAMPLERTYPE_Masks, texture=&LavaGround_M}.G
   B=TextureSample(UVs=Panner(Coordinate=@n71.YX, Speed=@n72)) {samplerType=SAMPLERTYPE_Masks, texture=&LavaGround_M}.G
  )
  n74 = MaterialFunctionCall(
   Transition Phase=Multiply(B=ScalarParameter() {defaultValue=0.7594, parameterName=TranstionHeightBlendEdge})
   Height Texture=Clamp(None=OneMinus(None=@n73)) {clampMode=CMODE_ClampMin}
   Contrast=ScalarParameter() {defaultValue=1.5753, parameterName=ContrastEdgeBlend}
  ) {materialFunction=&HeightLerp}
  Reroute(None=LinearInterpolate(B=LinearInterpolate(A=@n66.RGB, B=@n67.RGB, Alpha=@n74.Alpha)))
  ScalarParameter() {defaultValue=1886.8712, parameterName=FadeDistance}
 ]}
