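# Translucent unlit fog material for the Vulcano environment set, named as the far variant of a box fog.
# The whole graph is a single MF_BoxHeightFog material function call with its static bool input set to true, using material attributes.
# Depth testing is disabled, so the fog is not hidden by geometry in front of its mesh.
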Material /Figment_S09/Environments/Sets/Vulcano/Materials/M_BoxFog_far
graph {
 properties {
  bDisableDepthTest=true
  bUseMaterialAttributes=true
  blendMode=BLEND_Translucent
  pixelDepthOffsetMode=PDOM_Legacy
  shadingModel=MSM_Unlit
  usedShadingModels=Unlit
 }
 nodes [MaterialFunctionCall!(In=StaticBool() {value=true}) {materialFunction=&/MF_BoxHeightFog}]
}
