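# Material function that fakes lava light and glow as a tinted translucent volume over the scene behind it.
# Opacity combines radial and height falloff around the object with downward-facing scene normals, using a flat brightness on forward and mobile.
# A static switch selects an outside variant with far fade, or an inside variant with near fade and camera fog.
# World position offset collapses the mesh onto the camera when its variant is unneeded.
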MaterialFunction /Figment_S09/Environments/Sets/Vulcano/Materials/MF_FakeLightLava
graph {nodes [
  n1 = VectorParameter() {parameterName=Color, defaultValue=<rgba 1 0.168 0.0312 1>}
  n2 = ScalarParameter() {defaultValue=5, parameterName=Brightness}
  n3 = ShadingPathSwitch(Default=@n2, Deferred=@n2, Forward=Multiply(A=@n2) {constB=0.35})
  n4 = Distance(A=ObjectPositionWS(), B=CameraPositionWS())
  n5 = ScalarParameter() {defaultValue=15000, parameterName=CameraFadeOffset}
  n6 = Clamp(None=Divide(A=Subtract(A=@n4, B=@n5), B=ScalarParameter() {defaultValue=10000, parameterName=CameraFadeLength}))
  n7 = OneMinus(None=@n6)
  n8 = MaterialFunctionCall() {materialFunction=&WorldPositionBehindTranslucency}
  n9 = ObjectPositionWS()
  n10 = Clamp(None=Divide(A=Distance(A=@n8.Result, B=@n9), B=ScalarParameter() {defaultValue=2706.8237, parameterName=LightRadius}))
  n11 = Subtract(A=@n8.Result, B=@n9)
  n12 = ScalarParameter() {defaultValue=2992.521, parameterName=LightDistance}
  n13 = OneMinus(None=Divide(A=ComponentMask(None=@n11) {b=true}, B=@n12))
  n14 = Multiply(
   A=Power(Base=Clamp(None=@n13), Exp=ScalarParameter() {defaultValue=3.289, parameterName=Falloff})
   B=ScalarParameter() {defaultValue=0.373, parameterName=Opacity}
  )
  n15 = Multiply(A=Constant() {r=-1}, B=ComponentMask(None=SceneTexture() {sceneTextureId=PPI_WorldNormal}.Color) {b=true})
  n16 = Multiply(A=@n14, B=LinearInterpolate(A=ScalarParameter() {defaultValue=0.1143, parameterName=MinLightTop}, Alpha=Clamp(None=@n15)))
  n17 = Multiply(A=@n14, B=ScalarParameter() {defaultValue=0.5, parameterName=LowSpecPlatformBrightness})
  n18 = Clamp(None=Multiply(A=OneMinus(None=@n10), B=ShadingPathSwitch(Default=@n16, Deferred=@n16, Forward=@n17, Mobile=@n17)))
  n19 = Divide(
   A=Distance(
    A=MaterialFunctionCall() {materialFunction=&WorldPositionBehindTranslucency}.Result
    B=CameraPositionWS()
   )
   B=ScalarParameter() {defaultValue=2332.3992, parameterName=NearFadeDistance}
  )
  n20 = CameraPositionWS()
  n21 = ObjectPositionWS()
  n22 = Divide(A=ComponentMask(None=Subtract(A=@n20, B=@n21)) {b=true}, B=@n12)
  n23 = Power(Base=Clamp(None=OneMinus(None=@n22)), Exp=ScalarParameter() {defaultValue=3.289, parameterName=Falloff})
  n24 = ScalarParameter() {defaultValue=0.373, parameterName=Opacity}
  n25 = Distance(A=@n21, B=@n20)
  n26 = ScalarParameter() {defaultValue=6318.645, parameterName=FogRadius}
  n27 = OneMinus(None=Clamp(None=Divide(A=@n25, B=@n26)))
  n28 = ScalarParameter() {defaultValue=0.09, parameterName=FogIntensity}
  n29 = Max(A=@n18, B=Multiply(A=Multiply(A=@n23, B=@n24), B=Multiply(A=@n27, B=@n28)))
  n30 = FunctionInput(Preview=StaticBool()) {bUsePreviewValueAsDefault=true, inputType=FunctionInput_StaticBool}
  n31 = StaticSwitch(True=Multiply(A=@n7, B=@n18), False=Multiply(A=Clamp(None=@n19), B=@n29), Value=@n30)
  n32 = Subtract(A=CameraPositionWS(), B=WorldPosition().XYZ)
  n33 = Floor(None=@n6)
  n34 = Distance(A=ObjectPositionWS(), B=CameraPositionWS())
  n35 = ScalarParameter() {defaultValue=3579.1655, parameterName=SphereRadius}
  n36 = Ceil(None=Clamp(None=Subtract(A=@n34, B=@n35)))
  n37 = OneMinus(None=@n36)
  n38 = StaticSwitch(True=Clamp(None=Add(A=@n33, B=@n37)), False=@n36, Value=@n30)
  FunctionOutput(None=MakeMaterialAttributes!(
    BaseColor
    Metallic
    Specular
    Roughness
    Anisotropy
    EmissiveColor=Multiply(A=@n1.RGB, B=@n3)
    Opacity=@n31
    OpacityMask
    Normal
    Tangent
    WorldPositionOffset=Multiply(A=@n32, B=@n38)
    SubsurfaceColor
    ClearCoat
    ClearCoatRoughness
    AmbientOcclusion
    Refraction
    CustomizedUVs_0
    CustomizedUVs_1
    CustomizedUVs_2
    CustomizedUVs_3
    CustomizedUVs_4
    CustomizedUVs_5
    CustomizedUVs_6
    CustomizedUVs_7
    PixelDepthOffset
    ShadingModel
    Displacement
   ))
  OneMinus(None=@n7)
  StaticSwitch(True=@n31, False=@n29, Value=@n30)
 ]}
