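# M_FakeLavaLightFar is an additive, unlit material for the Vulcano environment set that fakes lava lighting, apparently a far-distance variant going by its name.
# All of its shading comes from the MF_FakeLightLava material function through material attributes, with a static bool parameter set to true on the function's input.
# Translucency vertex fog is turned off.
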Material /Figment_S09/Environments/Sets/Vulcano/Materials/M_FakeLavaLightFar
graph {
 properties {
  bUseMaterialAttributes=true
  bUseTranslucencyVertexFog=false
  blendMode=BLEND_Additive
  pixelDepthOffsetMode=PDOM_Legacy
  shadingModel=MSM_Unlit
  usedShadingModels=Unlit
 }
 nodes [MaterialFunctionCall!(In=StaticBoolParameter() {defaultValue=true}) {materialFunction=&/MF_FakeLightLava}]
}
