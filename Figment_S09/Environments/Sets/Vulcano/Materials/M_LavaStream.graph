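# Unlit translucent material for flowing lava streams in the Vulcano environment set, enabled for spline meshes.
# One panned mask texture drives the look: its green channel blends dark and bright HDR stream colors.
# Its red channel, scaled by an opacity parameter, clamped and multiplied by vertex color blue, forms an opacity mask.
# A MaterialProxyReplace node swaps in a brightness-multiplied color for baked HLOD proxies.
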Material /Figment_S09/Environments/Sets/Vulcano/Materials/M_LavaStream
graph {
 properties {bUsedWithSplineMeshes=true, blendMode=BLEND_Translucent, pixelDepthOffsetMode=PDOM_Legacy, shadingModel=MSM_Unlit, usedShadingModels=Unlit}
 nodes [
  n1 = ScalarParameter() {parameterName=PanSpeedStream2_X}
  n2 = ScalarParameter() {defaultValue=-0.05, parameterName=PanSpeedStream2_Y}
  n3 = TextureSample(UVs=Panner(Coordinate=TextureCoordinate(), Speed=AppendVector(A=@n1, B=@n2))) {samplerType=SAMPLERTYPE_Masks, texture=/Game/Environments/Sets/Vulcano/Textures/LavaStream_M}
  n4 = LinearInterpolate(
   A=VectorParameter() {parameterName=StreamColor_Dark, defaultValue=<rgba 0 0 0 1>}.RGB
   B=VectorParameter() {parameterName=StreamColor, defaultValue=<rgba 5 1.1178 0.4948 1>}.RGB
   Alpha=@n3.G
  )
  MaterialProxyReplace(Realtime=@n4, MaterialProxy=Multiply(A=@n4, B=ScalarParameter() {defaultValue=5, parameterName=HlodBirghtness}))
  n5 = ScalarParameter() {defaultValue=1, parameterName=Opcacity}
  Multiply(A=Clamp(None=Multiply(A=@n3.R, B=@n5)), B=VertexColor().B)
  VertexColor()
 ]
}
